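# Physics asset for the female medium Animal Jacket 01 body, set up for secondary motion on its scarf.
# Two box-body chains, six left and five right, branch from a small scarf master capsule constrained to spine_05, with twist locked at every scarf joint.
# Capsules on the pelvis, spine and upper arms form a torso proxy, linked by constraints with free swing and twist.
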PhysicsAsset /BRCosmetics/Characters/Player/Female/Medium/Bodies/F_MED_Animal_Jacket_01/Meshes/F_MED_Animal_Jacket_Physics
bodies {
 dyn_scarf_01_l [{shapeType=Box, extentX=7.6056, extentY=3.6979, extentZ=6, center=<2.8762 -0.0001 0>}]
 dyn_scarf_01_r [{shapeType=Box, extentX=5, extentY=2, extentZ=6.4278, center=<2 0 0>}]
 dyn_scarf_02_l [{shapeType=Box, extentX=7.8831, extentY=3.6979, extentZ=6, center=<3.5374 0 0>}]
 dyn_scarf_02_r [{shapeType=Box, extentX=5, extentY=3.1114, extentZ=6.4278, center=<2.995 0 0>}]
 dyn_scarf_03_l [{shapeType=Box, extentX=6.5505, extentY=3.4774, extentZ=6, center=<2.548 0 0>}]
 dyn_scarf_03_r [{shapeType=Box, extentX=5, extentY=3.1114, extentZ=6.4278, center=<2 0 0>}]
 dyn_scarf_04_l [{shapeType=Box, extentX=5.8413, extentY=2.7176, extentZ=6, center=<2 0 0>}]
 dyn_scarf_04_r [{shapeType=Box, extentX=6.3544, extentY=2.7577, extentZ=6.4278, center=<2.363 0 0>}]
 dyn_scarf_05_l [{shapeType=Box, extentX=5.8413, extentY=2, extentZ=6, center=<2 0 0>}]
 dyn_scarf_05_r [{shapeType=Box, extentX=5.2118, extentY=3.8201, extentZ=6.7918, center=<2.4439 0 0.0001>}]
 dyn_scarf_06_l [{shapeType=Box, extentX=5.8413, extentY=2, extentZ=6, center=<2 0 0>}]
 dyn_scarf_master [{shapeType=Capsule, length=0.505, radius=0.505, rotation=<rot 0 -90 0>}]
 pelvis [{shapeType=Capsule, length=12.1609, radius=13.6354}]
 spine_01 [{shapeType=Capsule, length=11.2754, radius=12.8513}]
 spine_02 [{shapeType=Capsule, length=11.7261, radius=12.969}]
 spine_03 [{shapeType=Capsule, length=9.4142, radius=12.7962, center=<0 1.0811 0>}]
 spine_04 [{shapeType=Capsule, length=11.3175, radius=11.2923, center=<0 1.0578 0>}]
 spine_05 [{shapeType=Capsule, length=17.346, radius=10.591}]
 upperarm_l [{shapeType=Capsule, length=14.2232, radius=4.7404, center=<10 0 2.267>, rotation=<rot 90 0 -179.9999>}]
 upperarm_r [{shapeType=Capsule, length=14.2232, radius=4.7404, center=<-10 -0.0001 -2.5009>, rotation=<rot 90 0 0>}]
}
constraints [
 {
  bone1Name=spine_01
  bone2Name=pelvis
  swing1LimitDegrees=45
  swing1Motion=Free
  swing2LimitDegrees=45
  swing2Motion=Free
  twistLimitDegrees=45
  twistMotion=Free
 }
 {
  bone1Name=spine_03
  bone2Name=spine_02
  swing1LimitDegrees=45
  swing1Motion=Free
  swing2LimitDegrees=45
  swing2Motion=Free
  twistLimitDegrees=45
  twistMotion=Free
 }
 {
  bone1Name=spine_04
  bone2Name=spine_03
  swing1LimitDegrees=45
  swing1Motion=Free
  swing2LimitDegrees=45
  swing2Motion=Free
  twistLimitDegrees=45
  twistMotion=Free
 }
 {
  bone1Name=spine_05
  bone2Name=spine_04
  swing1LimitDegrees=45
  swing1Motion=Free
  swing2LimitDegrees=45
  swing2Motion=Free
  twistLimitDegrees=45
  twistMotion=Free
 }
 {
  bone1Name=spine_02
  bone2Name=spine_01
  swing1LimitDegrees=45
  swing1Motion=Free
  swing2LimitDegrees=45
  swing2Motion=Free
  twistLimitDegrees=45
  twistMotion=Free
 }
 {bone1Name=dyn_scarf_master, bone2Name=spine_05, swing1LimitDegrees=45, swing2LimitDegrees=45, twistLimitDegrees=45, twistMotion=Locked}
 {bone1Name=dyn_scarf_01_l, bone2Name=dyn_scarf_master, swing1LimitDegrees=45, swing2LimitDegrees=45, twistLimitDegrees=45, twistMotion=Locked}
 {bone1Name=dyn_scarf_02_l, bone2Name=dyn_scarf_01_l, swing1LimitDegrees=45, swing2LimitDegrees=45, twistLimitDegrees=45, twistMotion=Locked}
 {bone1Name=dyn_scarf_03_l, bone2Name=dyn_scarf_02_l, swing1LimitDegrees=45, swing2LimitDegrees=45, twistLimitDegrees=45, twistMotion=Locked}
 {bone1Name=dyn_scarf_04_l, bone2Name=dyn_scarf_03_l, swing1LimitDegrees=45, swing2LimitDegrees=45, twistLimitDegrees=45, twistMotion=Locked}
 {bone1Name=dyn_scarf_05_l, bone2Name=dyn_scarf_04_l, swing1LimitDegrees=45, swing2LimitDegrees=45, twistLimitDegrees=45, twistMotion=Locked}
 {bone1Name=dyn_scarf_06_l, bone2Name=dyn_scarf_05_l, swing1LimitDegrees=45, swing2LimitDegrees=45, twistLimitDegrees=45, twistMotion=Locked}
 {bone1Name=dyn_scarf_01_r, bone2Name=dyn_scarf_master, swing1LimitDegrees=45, swing2LimitDegrees=45, twistLimitDegrees=45, twistMotion=Locked}
 {bone1Name=dyn_scarf_02_r, bone2Name=dyn_scarf_01_r, swing1LimitDegrees=45, swing2LimitDegrees=45, twistLimitDegrees=45, twistMotion=Locked}
 {bone1Name=dyn_scarf_03_r, bone2Name=dyn_scarf_02_r, swing1LimitDegrees=45, swing2LimitDegrees=45, twistLimitDegrees=45, twistMotion=Locked}
 {bone1Name=dyn_scarf_04_r, bone2Name=dyn_scarf_03_r, swing1LimitDegrees=45, swing2LimitDegrees=45, twistLimitDegrees=45, twistMotion=Locked}
 {bone1Name=dyn_scarf_05_r, bone2Name=dyn_scarf_04_r, swing1LimitDegrees=45, swing2LimitDegrees=45, twistLimitDegrees=45, twistMotion=Locked}
 {
  bone1Name=upperarm_r
  bone2Name=spine_05
  swing1LimitDegrees=45
  swing1Motion=Free
  swing2LimitDegrees=45
  swing2Motion=Free
  twistLimitDegrees=45
  twistMotion=Free
 }
 {
  bone1Name=upperarm_l
  bone2Name=spine_05
  swing1LimitDegrees=45
  swing1Motion=Free
  swing2LimitDegrees=45
  swing2Motion=Free
  twistLimitDegrees=45
  twistMotion=Free
 }
]
