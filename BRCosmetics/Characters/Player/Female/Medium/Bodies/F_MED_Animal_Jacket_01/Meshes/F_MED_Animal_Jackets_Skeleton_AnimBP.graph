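# Post-process style AnimBlueprint for the female medium Animal Jacket cosmetic, adding secondary motion to its scarf and rear skirt.
# It copies the attached parent mesh's pose, then chains damped, angular-spring AnimDynamics nodes along five left and five right scarf bones.
# The rear skirt bone is constrained to pelvis and thighs, then given an additive roll whose alpha derives from clamped 2D pawn speed.
# A RigidBody node is present but unconnected.
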AnimBlueprint /BRCosmetics/Characters/Player/Female/Medium/Bodies/F_MED_Animal_Jacket_01/Meshes/F_MED_Animal_Jackets_Skeleton_AnimBP
target_skeleton=&/F_MED_Animal_Jackets_Skeleton
graphs [
 AnimGraph {nodes [
   RigidBody() {
    overridePhysicsAsset=&/F_MED_Animal_Jacket_Physics
    simulationSpace=BaseBoneSpace
    alphaBoolBlend {blendOption=Linear}
    baseBoneRef {boneName=root}
    componentAppliedLinearAccClamp=<50 50 50>
    componentLinearAccScale=<1 1 1>
    componentLinearVelScale=<1 1 1>
   }
   n1 = Anim Dynamics - Bone: dyn_scarf_01_l(ComponentPose=Local To Component(LocalPose=Copy Pose From Mesh() {bUseAttachedParent=true}), Alpha=1, ExternalForce="0,0,0") {
    angularDampingOverride=0.999
    angularSpringConstant=2500
    bAngularSpring=true
    bGravityOverrideInSimSpace=true
    bOverrideAngularDamping=true
    gravityScale=0.35
    simulationSpace=RootRelative
    alphaBoolBlend {blendOption=Linear}
    boundBone {boneName=dyn_scarf_01_l}
    componentAppliedLinearAccClamp=<10 10 10>
    componentLinearAccScale=<0.2 0.2 0.2>
    componentLinearVelScale=<0.2 0.2 0.2>
    physicsBodyDefinitions [{
      sphereCollisionRadius=0
      boundBone {boneName=dyn_scarf_01_l}
      boxExtents=<7 30 30>
      constraintSetup {angularLimitsMax=<0 20 20>, angularLimitsMin=<0 -20 0>, angularTarget=<0 0 0>}
      localJointOffset=<7 0 0>
     }]
   }
   n2 = Anim Dynamics - Bone: dyn_scarf_02_l(ComponentPose=@n1, Alpha=1, ExternalForce="0,0,0") {
    angularDampingOverride=0.999
    angularSpringConstant=2500
    bAngularSpring=true
    bGravityOverrideInSimSpace=true
    bOverrideAngularDamping=true
    gravityScale=0.35
    simulationSpace=RootRelative
    alphaBoolBlend {blendOption=Linear}
    boundBone {boneName=dyn_scarf_02_l}
    componentAppliedLinearAccClamp=<10 10 10>
    componentLinearAccScale=<0.2 0.2 0.2>
    componentLinearVelScale=<0.2 0.2 0.2>
    physicsBodyDefinitions [{
      sphereCollisionRadius=0
      boundBone {boneName=dyn_scarf_02_l}
      boxExtents=<7 30 30>
      constraintSetup {angularLimitsMax=<0 20 20>, angularLimitsMin=<0 -20 0>, angularTarget=<0 0 0>}
      localJointOffset=<7 0 0>
     }]
   }
   n3 = Anim Dynamics - Bone: dyn_scarf_03_l(ComponentPose=@n2, Alpha=1, ExternalForce="0,0,0") {
    angularDampingOverride=0.999
    angularSpringConstant=2500
    bAngularSpring=true
    bGravityOverrideInSimSpace=true
    bOverrideAngularDamping=true
    gravityScale=0.35
    simulationSpace=RootRelative
    alphaBoolBlend {blendOption=Linear}
    boundBone {boneName=dyn_scarf_03_l}
    componentAppliedLinearAccClamp=<10 10 10>
    componentLinearAccScale=<0.2 0.2 0.2>
    componentLinearVelScale=<0.2 0.2 0.2>
    physicsBodyDefinitions [{
      sphereCollisionRadius=0
      boundBone {boneName=dyn_scarf_03_l}
      boxExtents=<7 30 30>
      constraintSetup {angularLimitsMax=<0 20 20>, angularLimitsMin=<0 -20 0>, angularTarget=<0 0 0>}
      localJointOffset=<7 0 0>
     }]
   }
   n4 = Anim Dynamics - Bone: dyn_scarf_04_l(ComponentPose=@n3, Alpha=1, ExternalForce="0,0,0") {
    angularDampingOverride=0.999
    angularSpringConstant=2500
    bAngularSpring=true
    bGravityOverrideInSimSpace=true
    bOverrideAngularDamping=true
    gravityScale=0.35
    simulationSpace=RootRelative
    alphaBoolBlend {blendOption=Linear}
    boundBone {boneName=dyn_scarf_04_l}
    componentAppliedLinearAccClamp=<10 10 10>
    componentLinearAccScale=<0.2 0.2 0.2>
    componentLinearVelScale=<0.2 0.2 0.2>
    physicsBodyDefinitions [{
      sphereCollisionRadius=0
      boundBone {boneName=dyn_scarf_04_l}
      boxExtents=<7 30 30>
      constraintSetup {angularLimitsMax=<0 20 20>, angularLimitsMin=<0 -20 0>, angularTarget=<0 0 0>}
      localJointOffset=<7 0 0>
     }]
   }
   n5 = Anim Dynamics - Bone: dyn_scarf_05_l(ComponentPose=@n4, Alpha=1, ExternalForce="0,0,0") {
    angularDampingOverride=0.999
    angularSpringConstant=2500
    bAngularSpring=true
    bGravityOverrideInSimSpace=true
    bOverrideAngularDamping=true
    gravityScale=0.35
    simulationSpace=RootRelative
    alphaBoolBlend {blendOption=Linear}
    boundBone {boneName=dyn_scarf_05_l}
    componentAppliedLinearAccClamp=<10 10 10>
    componentLinearAccScale=<0.2 0.2 0.2>
    componentLinearVelScale=<0.2 0.2 0.2>
    physicsBodyDefinitions [{
      sphereCollisionRadius=0
      boundBone {boneName=dyn_scarf_05_l}
      boxExtents=<7 30 30>
      constraintSetup {angularLimitsMax=<0 20 20>, angularLimitsMin=<0 -20 0>, angularTarget=<0 0 0>}
      localJointOffset=<7 0 0>
     }]
   }
   n6 = Anim Dynamics - Bone: dyn_scarf_01_r(ComponentPose=@n5, Alpha=1, ExternalForce="0,0,0") {
    angularDampingOverride=0.999
    angularSpringConstant=2500
    bAngularSpring=true
    bGravityOverrideInSimSpace=true
    bOverrideAngularDamping=true
    gravityScale=0.35
    simulationSpace=RootRelative
    alphaBoolBlend {blendOption=Linear}
    boundBone {boneName=dyn_scarf_01_r}
    componentAppliedLinearAccClamp=<10 10 10>
    componentLinearAccScale=<0.2 0.2 0.2>
    componentLinearVelScale=<0.2 0.2 0.2>
    physicsBodyDefinitions [{
      sphereCollisionRadius=0
      boundBone {boneName=dyn_scarf_01_r}
      boxExtents=<8 30 30>
      constraintSetup {angularLimitsMax=<0 20 20>, angularLimitsMin=<0 -20 0>, angularTarget=<0 0 0>}
      localJointOffset=<8 0 0>
     }]
   }
   n7 = Anim Dynamics - Bone: dyn_scarf_02_r(ComponentPose=@n6, Alpha=1, ExternalForce="0,0,0") {
    angularDampingOverride=0.999
    angularSpringConstant=2500
    bAngularSpring=true
    bGravityOverrideInSimSpace=true
    bOverrideAngularDamping=true
    gravityScale=0.35
    simulationSpace=RootRelative
    alphaBoolBlend {blendOption=Linear}
    boundBone {boneName=dyn_scarf_02_r}
    componentAppliedLinearAccClamp=<10 10 10>
    componentLinearAccScale=<0.2 0.2 0.2>
    componentLinearVelScale=<0.2 0.2 0.2>
    physicsBodyDefinitions [{
      sphereCollisionRadius=0
      boundBone {boneName=dyn_scarf_02_r}
      boxExtents=<8 30 30>
      constraintSetup {angularLimitsMax=<0 20 20>, angularLimitsMin=<0 -20 0>, angularTarget=<0 0 0>}
      localJointOffset=<8 0 0>
     }]
   }
   n8 = Anim Dynamics - Bone: dyn_scarf_03_r(ComponentPose=@n7, Alpha=1, ExternalForce="0,0,0") {
    angularDampingOverride=0.999
    angularSpringConstant=2500
    bAngularSpring=true
    bGravityOverrideInSimSpace=true
    bOverrideAngularDamping=true
    gravityScale=0.35
    simulationSpace=RootRelative
    alphaBoolBlend {blendOption=Linear}
    boundBone {boneName=dyn_scarf_03_r}
    componentAppliedLinearAccClamp=<10 10 10>
    componentLinearAccScale=<0.2 0.2 0.2>
    componentLinearVelScale=<0.2 0.2 0.2>
    physicsBodyDefinitions [{
      sphereCollisionRadius=0
      boundBone {boneName=dyn_scarf_03_r}
      boxExtents=<8 30 30>
      constraintSetup {angularLimitsMax=<0 20 20>, angularLimitsMin=<0 -20 0>, angularTarget=<0 0 0>}
      localJointOffset=<8 0 0>
     }]
   }
   n9 = Anim Dynamics - Bone: dyn_scarf_04_r(ComponentPose=@n8, Alpha=1, ExternalForce="0,0,0") {
    angularDampingOverride=0.999
    angularSpringConstant=2500
    bAngularSpring=true
    bGravityOverrideInSimSpace=true
    bOverrideAngularDamping=true
    gravityScale=0.35
    simulationSpace=RootRelative
    alphaBoolBlend {blendOption=Linear}
    boundBone {boneName=dyn_scarf_04_r}
    componentAppliedLinearAccClamp=<10 10 10>
    componentLinearAccScale=<0.2 0.2 0.2>
    componentLinearVelScale=<0.2 0.2 0.2>
    physicsBodyDefinitions [{
      sphereCollisionRadius=0
      boundBone {boneName=dyn_scarf_04_r}
      boxExtents=<8 30 30>
      constraintSetup {angularLimitsMax=<0 20 20>, angularLimitsMin=<0 -20 0>, angularTarget=<0 0 0>}
      localJointOffset=<8 0 0>
     }]
   }
   n10 = Anim Dynamics - Bone: dyn_scarf_05_r(ComponentPose=@n9, Alpha=1, ExternalForce="0,0,0") {
    angularDampingOverride=0.999
    angularSpringConstant=2500
    bAngularSpring=true
    bGravityOverrideInSimSpace=true
    bOverrideAngularDamping=true
    gravityScale=0.35
    simulationSpace=RootRelative
    alphaBoolBlend {blendOption=Linear}
    boundBone {boneName=dyn_scarf_05_r}
    componentAppliedLinearAccClamp=<10 10 10>
    componentLinearAccScale=<0.2 0.2 0.2>
    componentLinearVelScale=<0.2 0.2 0.2>
    physicsBodyDefinitions [{
      sphereCollisionRadius=0
      boundBone {boneName=dyn_scarf_05_r}
      boxExtents=<8 30 30>
      constraintSetup {angularLimitsMax=<0 20 20>, angularLimitsMin=<0 -20 0>, angularTarget=<0 0 0>}
      localJointOffset=<8 0 0>
     }]
   }
   n11 = "Constraint - dyn_skirt_rear (PPP)"!(ComponentPose=@n10, Alpha=1, ConstraintWeights_0=1, ConstraintWeights_1:float=1, ConstraintWeights_2:float=1) -> (Pose) {
    alphaBoolBlend {blendOption=Linear}
    boneToModify {boneName=dyn_skirt_rear}
    constraintSetup [{targetBone {boneName=pelvis}}, {targetBone {boneName=thigh_l}}, {targetBone {boneName=thigh_r}}]
    constraintWeights [1,1,1]
   }
   n12 = "Transform (Modify) Bone - Bone: dyn_skirt_rear"(
    ComponentPose=@n11
    Alpha=Get skirtCorrectiveAlpha() -> (skirtCorrectiveAlpha:double)
    Translation="0,0,0"
    Rotation="0.000,0.000,30.000"
    Scale="1,1,1"
   ) {rotationMode=BMM_Additive, alphaBoolBlend {blendOption=Linear}, boneToModify {boneName=dyn_skirt_rear}, rotation {roll=30}}
   Output Pose(Result=Component To Local(ComponentPose=@n12)) {layerGroup=None}
  ]}
 EventGraph {nodes [
   Event BlueprintUpdateAnimation()
   > Set skirtCorrectiveAlpha(skirtCorrectiveAlpha:double=MapRangeClamped(Value=Get PawnSpeed2D(), InRangeA=250, InRangeB=400)) -> (Output_Get:double)
   TryGetPawnOwner()
  ]}
]
variables [skirtCorrectiveAlpha {type=number}]
