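# Physics asset for the face accessory of the female medium Animal Jacket cosmetic, simulating its left and right ears.
# Each ear is a two-segment capsule chain constrained to a single rounded head body.
# All joints lock twist, and the outer ear joints also lock the second swing axis, so they bend on one axis like hinges.
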PhysicsAsset /BRCosmetics/Characters/Player/Female/Medium/Bodies/F_MED_Animal_Jacket_01/Meshes/Parts/F_MED_Animal_Jackets_FaceAcc_Physics
bodies {
 dyn_ear_01_l [{shapeType=Capsule, length=5, radius=4, center=<5 0 0>, rotation=<rot 90 0 0>}]
 dyn_ear_01_r [{shapeType=Capsule, length=5, radius=4, center=<5 0 0>, rotation=<rot 90 0 0>}]
 dyn_ear_02_l [{shapeType=Capsule, length=5, radius=4, center=<5 0 0>, rotation=<rot 90 0 0>}]
 dyn_ear_02_r [{shapeType=Capsule, length=5, radius=4, center=<5 0 0>, rotation=<rot 90 0 0>}]
 head [{shapeType=Capsule, length=0, radius=10}]
}
constraints [
 {bone1Name=dyn_ear_01_l, bone2Name=head, swing1LimitDegrees=45, swing2LimitDegrees=10, twistLimitDegrees=45, twistMotion=Locked}
 {
  bone1Name=dyn_ear_02_l
  bone2Name=dyn_ear_01_l
  swing1LimitDegrees=45
  swing2LimitDegrees=45
  swing2Motion=Locked
  twistLimitDegrees=45
  twistMotion=Locked
 }
 {bone1Name=dyn_ear_01_r, bone2Name=head, swing1LimitDegrees=45, swing2LimitDegrees=10, twistLimitDegrees=45, twistMotion=Locked}
 {
  bone1Name=dyn_ear_02_r
  bone2Name=dyn_ear_01_r
  swing1LimitDegrees=45
  swing2LimitDegrees=45
  swing2Motion=Locked
  twistLimitDegrees=45
  twistMotion=Locked
 }
]
