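# Anim Blueprint for the female animal jacket face accessory, copying the attached parent's pose and adding secondary motion to its ears and ear tag.
# A RigidBody pass feeds AnimDynamics on the ear tag, additive right-ear rotations, then a right-ear RigidBody pass.
# The update event derives the rotation and right-ear physics alphas from targeting state.
# When the playlist uses custom character parts, output switches to the copied pose with the head bone shrunk to near zero.
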AnimBlueprint /BRCosmetics/Characters/Player/Female/Medium/Bodies/F_MED_Animal_Jacket_01/Meshes/Parts/F_MED_Animal_Jackets_FaceAcc_Skeleton_AnimBP
target_skeleton=&/F_MED_Animal_Jackets_FaceAcc_Skeleton
graphs [
 AnimGraph {nodes [
   n1 = "Transform (Modify) Bone - Bone: head"!(ComponentPose=Local To Component(LocalPose=Use cached pose 'PreScale Cache'()), Alpha=1) -> (Pose) {scaleMode=BMM_Replace, alphaBoolBlend {blendOption=Linear}, boneToModify {boneName=head}, scale=<0.01 0.01 0.01>}
   n2 = RigidBody(ComponentPose=Local To Component(LocalPose=Use cached pose 'PreScale Cache'())) {
    bClampLinearTranslationLimitToRefPose=true
    bOverrideWorldGravity=true
    overridePhysicsAsset=&/F_MED_Animal_Jackets_FaceAcc_Physics
    simulationSpace=BaseBoneSpace
    alphaBoolBlend {blendOption=Linear}
    alphaScaleBiasClamp {bInterpResult=true}
    baseBoneRef {boneName=root}
    componentAppliedLinearAccClamp=<1 1 1>
    overrideWorldGravity {z=-100}
   }
   n3 = Anim Dynamics - Bone: dyn_eartag_01(ComponentPose=@n2, Alpha=1, ExternalForce="0,0,0") {
    angularDampingOverride=0.999
    bGravityOverrideInSimSpace=true
    bOverrideAngularDamping=true
    gravityScale=0.35
    simulationSpace=RootRelative
    alphaBoolBlend {blendOption=Linear}
    boundBone {boneName=dyn_eartag_01}
    componentLinearAccScale {z=-30}
    componentLinearVelScale {z=30}
    physicsBodyDefinitions [{
      sphereCollisionRadius=0
      boundBone {boneName=dyn_eartag_01}
      boxExtents=<2 5 5>
      constraintSetup {angularLimitsMax=<0 0 0>, angularLimitsMin=<0 0 0>, angularTarget=<0 0 0>}
      localJointOffset=<2 0 0>
     }]
   }
   n4 = Anim Dynamics - Bone: dyn_eartag_02(ComponentPose=@n3, Alpha=1, ExternalForce="0,0,0") {
    angularDampingOverride=0.999
    bGravityOverrideInSimSpace=true
    bOverrideAngularDamping=true
    gravityScale=0.35
    simulationSpace=RootRelative
    alphaBoolBlend {blendOption=Linear}
    boundBone {boneName=dyn_eartag_02}
    physicsBodyDefinitions [{
      sphereCollisionRadius=0
      boundBone {boneName=dyn_eartag_02}
      boxExtents=<2 5 5>
      constraintSetup {angularLimitsMax=<0 20 45>, angularLimitsMin=<0 -20 -45>, angularTarget=<0 0 0>}
      localJointOffset=<2 0 0>
     }]
   }
   n5 = Get targetingAlpha() -> (targetingAlpha:double)
   n6 = "Transform (Modify) Bone - Bone: dyn_ear_01_r"(ComponentPose=@n4, Alpha=@n5, Translation="0,0,0", Rotation="11.814,-26.934,22.220", Scale="1,1,1") {
    rotationMode=BMM_Additive
    alphaBoolBlend {blendOption=Linear}
    alphaScaleBiasClamp {bInterpResult=true}
    boneToModify {boneName=dyn_ear_01_r}
    rotation=<rot 11.814 -26.9337 22.2197>
   }
   n7 = "Transform (Modify) Bone - Bone: dyn_ear_02_r"(ComponentPose=@n6, Alpha=@n5, Translation="0,0,0", Rotation="9.847,-10.155,-1.754", Scale="1,1,1") {
    rotationMode=BMM_Additive
    alphaBoolBlend {blendOption=Linear}
    alphaScaleBiasClamp {bInterpResult=true}
    boneToModify {boneName=dyn_ear_02_r}
    rotation=<rot 9.847 -10.1553 -1.754>
   }
   n8 = RigidBody(ComponentPose=@n7, Alpha=Get targetingPhysicsAlpha() -> (targetingPhysicsAlpha:double)) {
    bClampLinearTranslationLimitToRefPose=true
    bOverrideWorldGravity=true
    overridePhysicsAsset=&/F_MED_Animal_Jackets_FaceAcc_RightEar_Physics
    simulationSpace=BaseBoneSpace
    alphaBoolBlend {blendOption=Linear}
    alphaScaleBiasClamp {bInterpResult=true}
    baseBoneRef {boneName=root}
    componentAppliedLinearAccClamp=<1 1 1>
    overrideWorldGravity {z=-100}
   }
   Output Pose(Result=Blend Poses by bool(
     BlendPose_0=Component To Local(ComponentPose=@n1)
     BlendPose_1=Component To Local(ComponentPose=@n8)
     BlendTime_0=0
     BlendTime_1=0
     bActiveValue=Get GetPlayerListUsesCustomCharacterParts() -> (GetPlayerListUsesCustomCharacterParts:bool)
    ) {blendType=Linear}) {layerGroup=None}
   Save cached pose 'PreScale Cache'(Pose=Copy Pose From Mesh() {bUseAttachedParent=true}) {cacheName=PreScale Cache}
  ]}
 EventGraph {nodes [
   Event BlueprintUpdateAnimation()
   n1 = SelectFloat(A=1, B=0, bPickA=Get bIsTargeting())
   > Set targetingAlpha(targetingAlpha:double=@n1) -> (Output_Get:double)
   > Set targetingPhysicsAlpha(targetingPhysicsAlpha:double=float - float(B=@n1) {is=title K2Node_CallFunction}) -> (Output_Get:double)
   > Set GetPlayerListUsesCustomCharacterParts(GetPlayerListUsesCustomCharacterParts:bool=GetPlaylistUsesCustomCharacterParts!() -> (ReturnValue:bool)) -> (Output_Get:bool)
  ]}
]
variables [
 GetPlayerListUsesCustomCharacterParts {type=boolean}
 targetingAlpha {type=number}
 targetingPhysicsAlpha {type=number}
]
